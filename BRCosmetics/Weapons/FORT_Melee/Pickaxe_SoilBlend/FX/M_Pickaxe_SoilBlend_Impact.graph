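# Unlit translucent, two-sided Niagara material for SoilBlend pickaxe impact effects on sprites and meshes.
# Scalar-driven lerps pick mask red or blue (glow), optional four-frame SubUV, linear wipe and SplineThicken vertex offset.
# Emissive tints the mask, plus screen-aligned smoke in non-glow mode, by two colors and particle color.
# Dynamic parameters drive height-lerped cloud-noise dissolve, spline width, wipe offset and frame.
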Material /BRCosmetics/Weapons/FORT_Melee/Pickaxe_SoilBlend/FX/M_Pickaxe_SoilBlend_Impact
graph {
 properties {
  bUsedWithNiagaraMeshParticles=true
  bUsedWithNiagaraSprites=true
  blendMode=BLEND_Translucent
  pixelDepthOffsetMode=PDOM_Legacy
  shadingModel=MSM_Unlit
  twoSided=true
  usedShadingModels=Unlit
 }
 nodes [
  Constant()
  n1 = DynamicParameter() {defaultValue=<rgba 1 3 -0.2 0>, paramNames [Dissolve, SplineWidth, LinearMask_Offset, SubUV_Frame]}
  n2 = StaticBool()
  LinearInterpolate(
   A=Constant()
   B=Multiply(
    A=ScalarParameter() {defaultValue=1, parameterName=Spline_ThickMul}
    B=MaterialFunctionCall(WidthBase=@n1.SplineWidth, WidthTip=@n1.SplineWidth, Expand U or V UV Channel=@n2) {materialFunction=&SplineThicken}.WorldPositionOffset
   )
   Alpha=ScalarParameter() {parameterName=Use_SplineThicken}
  )
  n3 = TextureCoordinate()
  n4 = AppendVector(A=Constant() {r=1}, B=Constant() {r=4})
  n5 = DynamicParameter() {defaultValue=<rgba 1 3 -0.2 0>, paramNames [Dissolve, SplineWidth, LinearMask_Offset, SubUV_Frame]}
  n6 = TextureSampleParameter2D(UVs=LinearInterpolate(
    A=TextureCoordinate()
    B=MaterialFunctionCall!(UVs=@n3, SubImages=@n4, Frame=@n5.SubUV_Frame) {materialFunction=/BRCosmetics/Animation/Game/MainPlayer/Emotes/TroutWrist/FX/Materials/SubUV_Function}.UVs
    Alpha=ScalarParameter() {parameterName=Use_SubUV}
   )) {parameterName=Texture, texture=&/T_Pickaxe_SoilBlend_ImpactMask}
  n7 = Multiply(A=@n6.R, B=@n6.R)
  n8 = ParticleColor()
  n9 = Multiply(A=@n8.A, B=@n6.B)
  n10 = ScalarParameter() {parameterName=Use glow}
  n11 = TextureSample(UVs=Multiply(A=MaterialFunctionCall!(Center Point World Position) {materialFunction=/Engine/Functions/Engine_MaterialFunctions02/UVs/WorldSpaceAlignedScreenCoordinates}.Offset and Scaled UVS) {constB=10}) {texture=/Game/Global/Textures/Default/Tileable_Patterns/T_Thick_Smoke_05}
  n12 = Add(
   A=Multiply(A=VectorParameter() {parameterName=Color1, defaultValue=<rgba 0 1 0.2211 1>}.RGB, B=LinearInterpolate(A=@n7, B=@n9, Alpha=@n10))
   B=LinearInterpolate(A=Multiply(A=@n11.RGB) {constB=0.2}, Alpha=ScalarParameter() {parameterName=Use glow}) {constB=0}
  )
  MaterialFunctionCall(Emissive=Multiply(A=VectorParameter() {parameterName=Color2, defaultValue=<rgba 0 8 2.0682 1>}.RGB, B=Multiply(A=@n12, B=@n8.RGB))) {materialFunction=&MF_FX_EmissiveAdjustments}
  n13 = LinearInterpolate(A=Add(A=@n6.R, B=@n6.R), B=@n6.B, Alpha=ScalarParameter() {parameterName=Use glow})
  n14 = DynamicParameter() {defaultValue=<rgba 1 3 -0.2 0>, paramNames [Dissolve, SplineWidth, LinearMask_Offset, SubUV_Frame]}
  n15 = TextureSample(UVs=MaterialFunctionCall(XY=TextureCoordinate() {vTiling=0.3}) {materialFunction=&Swizzle}.YX) {samplerType=SAMPLERTYPE_LinearGrayscale, texture=/Game/Athena/Prototype/Terrain/T_CloudCoverage}
  n16 = Saturate(None=Multiply(A=MaterialFunctionCall(Transition Phase=@n14.Dissolve, Height Texture=@n15.R) {materialFunction=&HeightLerp}.Alpha) {constB=3})
  n17 = Constant()
  n18 = ComponentMask(None=Add(A=TextureCoordinate(), B=AppendVector(A=@n14.LinearMask_Offset, B=@n17))) {r=true}
  n19 = Power(Base=Multiply(A=OneMinus(None=@n18)) {constB=5}) {constExponent=5}
  Multiply(
   A=Multiply(A=@n8.A, B=Multiply(A=@n13, B=@n16))
   B=LinearInterpolate(A=Constant() {r=1}, B=Saturate(None=@n19), Alpha=ScalarParameter() {parameterName=Use_LinearMask})
  )
 ]
}
