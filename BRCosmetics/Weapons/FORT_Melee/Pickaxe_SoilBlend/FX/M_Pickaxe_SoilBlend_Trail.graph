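# Translucent unlit, two-sided Niagara trail material for the SoilBlend pickaxe, colored by particle color through an emissive adjustment function.
# Opacity blends, via Use_SecondTrail, between two panned trail textures with edge fades and the default noise-distorted flame/energy mask, which has a switchable caustics breakup.
# The second trail dissolves through a dynamic-parameter HeightLerp.
# Particle alpha and view-angle fade scale the result, with depth fade skipped on forward and mobile paths.
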Material /BRCosmetics/Weapons/FORT_Melee/Pickaxe_SoilBlend/FX/M_Pickaxe_SoilBlend_Trail
graph {
 properties {
  bAutomaticallySetUsageInEditor=false
  bUsedWithNiagaraMeshParticles=true
  bUsedWithNiagaraRibbons=true
  bUsedWithNiagaraSprites=true
  blendMode=BLEND_Translucent
  shadingModel=MSM_Unlit
  twoSided=true
  usedShadingModels=Unlit
 }
 nodes [
  n1 = ParticleColor()
  MaterialFunctionCall(Emissive=@n1.RGB) {materialFunction=&MF_FX_EmissiveAdjustments}
  NamedRerouteDeclaration(None=Multiply(A=Constant2Vector() {r=-1}, B=ScalarParameter() {defaultValue=1.5, parameterName=Texture_Speed_Master})) {name=Texture_Speed_Direction, nodeColor=<rgba 1 0 0.75 1>}
  n2 = TextureSample(UVs=Panner(Coordinate=TextureCoordinate() {uTiling=0.8, vTiling=0.5}) {speedX=-0.1, speedY=0.2}) {samplerType=SAMPLERTYPE_Masks, texture=/Game/Weapons/WeaponSkins/Wraps/Textures/Mash/Mash_Noise}
  NamedRerouteDeclaration(None=Multiply(A=ComponentMask(None=@n2.RGB) {g=true, r=true}) {constB=0.01}) {name=Noise, nodeColor=<rgba 1 0.746 0 1>}
  n3 = TextureCoordinate() {uTiling=2}
  n4 = AppendVector(A=ScalarParameter() {defaultValue=1.5, parameterName=TrailTexture01_U}, B=ScalarParameter() {defaultValue=1, parameterName=TrailTexture01_V})
  n5 = Panner(Coordinate=Add(A=Multiply(A=@n3, B=@n4), B=NamedRerouteUsage()), Speed=NamedRerouteUsage()) {speedX=-1}
  n6 = TextureCoordinate() {uTiling=2}
  n7 = AppendVector(A=ScalarParameter() {defaultValue=1, parameterName=TrailTexture02_U}, B=ScalarParameter() {defaultValue=1, parameterName=TrailTexture02_V})
  n8 = NamedRerouteUsage()
  n9 = TextureSampleParameter2D(UVs=Panner(Coordinate=Multiply(A=@n6, B=@n7), Speed=Multiply(A=@n8) {constB=0.3}) {speedX=-1}) {
   parameterName=TrailTexture02
   samplerType=SAMPLERTYPE_Grayscale
   texture=/Game/Effects/Fort_Effects/Textures/Smoke/T_Smoke_Trail_Soft_NoClamp
  }
  n10 = Multiply(A=@n9.R) {constB=0.7}
  n11 = ComponentMask(None=TextureCoordinate()) {r=true}
  n12 = Add(A=@n11, B=ScalarParameter() {defaultValue=-0.05, group="Fade Leading && Trailing Edges", parameterName=Y_Leading Bias, sortPriority=2})
  n13 = ScalarParameter() {defaultValue=3, group="Fade Leading && Trailing Edges", parameterName=Y_Leading Scale}
  n14 = LinearInterpolate(
   B=Saturate(None=Multiply(A=@n12, B=@n13) {constB=3})
   Alpha=ScalarParameter() {defaultValue=1, group="Fade Leading && Trailing Edges", parameterName=Y_Fade Leading Edge, sortPriority=4}
  ) {constA=1}
  n15 = Add(
   A=OneMinus(None=@n11)
   B=ScalarParameter() {defaultValue=-0.05, group="Fade Leading && Trailing Edges", parameterName=Y_Trailing Bias, sortPriority=3}
  )
  n16 = ScalarParameter() {defaultValue=2, group="Fade Leading && Trailing Edges", parameterName=Y_Trailing Scale, sortPriority=1}
  n17 = LinearInterpolate(
   B=Saturate(None=Multiply(A=@n15, B=@n16) {constB=2})
   Alpha=ScalarParameter() {defaultValue=1, group="Fade Leading && Trailing Edges", parameterName=Y_Fade Trailing Edge, sortPriority=5}
  ) {constA=1}
  n18 = ComponentMask(None=TextureCoordinate()) {g=true}
  n19 = Add(A=@n18, B=ScalarParameter() {group="Fade Leading && Trailing Edges", parameterName=X_Leading Bias, sortPriority=2})
  n20 = ScalarParameter() {defaultValue=3, group="Fade Leading && Trailing Edges", parameterName=X_Leading Scale}
  n21 = LinearInterpolate(
   B=Saturate(None=Multiply(A=@n19, B=@n20) {constB=3})
   Alpha=ScalarParameter() {defaultValue=1, group="Fade Leading && Trailing Edges", parameterName=X_Fade Leading Edge, sortPriority=4}
  ) {constA=1}
  n22 = Add(A=OneMinus(None=@n18), B=ScalarParameter() {group="Fade Leading && Trailing Edges", parameterName=X_Trailing Bias, sortPriority=3})
  n23 = ScalarParameter() {defaultValue=3, group="Fade Leading && Trailing Edges", parameterName=X_Trailing Scale, sortPriority=1}
  n24 = LinearInterpolate(
   B=Saturate(None=Multiply(A=@n22, B=@n23) {constB=2})
   Alpha=ScalarParameter() {defaultValue=1, group="Fade Leading && Trailing Edges", parameterName=X_Fade Trailing Edge, sortPriority=5}
  ) {constA=1}
  n25 = Multiply(
   A=Add(
    A=TextureSampleParameter2D(UVs=@n5) {parameterName=TrailTexture01, texture=/BRCosmetics/Weapons/FORT_Melee/Pickaxe_Sunlit/FX/T_Pickaxe_Sunlit_Trail_01}.R
    B=Power(Base=@n10) {constExponent=1}
   )
   B=Multiply(A=Multiply(A=@n14, B=@n17), B=Multiply(A=@n21, B=@n24))
  )
  n26 = ComponentMask(None=TextureCoordinate() {vTiling=0.45}) {r=true}
  n27 = Saturate(None=Power(Base=Multiply(A=@n26) {constB=20}) {constExponent=1})
  n28 = OneMinus(None=@n26)
  n29 = Saturate(None=Power(Base=Multiply(A=@n28) {constB=2.8}) {constExponent=1})
  n30 = Power(Base=Multiply(A=@n26) {constB=5}) {constExponent=1.5}
  n31 = Multiply(A=TextureCoordinate(), B=ScalarParameter() {defaultValue=1, parameterName=Texture_Tile_Master})
  n32 = AppendVector(A=ScalarParameter() {defaultValue=0.7, parameterName=MaskTextureU}, B=ScalarParameter() {defaultValue=1, parameterName=MaskTextureV})
  n33 = TextureCoordinate() {uTiling=0.5, vTiling=0.8}
  n34 = ComponentMask(None=TextureSample(UVs=Panner(Coordinate=@n33) {speedX=-0.1}) {samplerType=SAMPLERTYPE_Masks, texture=&Mash_Noise}.RGB) {g=true, r=true}
  n35 = ScalarParameter() {defaultValue=0.05, parameterName=Noise_Intensity}
  n36 = ScalarParameter() {defaultValue=-1, parameterName=MaskTexture_SpeedX}
  n37 = ScalarParameter() {parameterName=MaskTexture_SpeedY}
  n38 = Panner(
   Coordinate=Add(A=Multiply(A=@n31, B=@n32), B=Multiply(A=@n34, B=@n35) {constB=0.3})
   Speed=Multiply(A=AppendVector(A=@n36, B=@n37), B=ScalarParameter() {defaultValue=1.5, parameterName=Texture_Speed_Master})
  ) {speedX=-1}
  n39 = Multiply(A=TextureSampleParameter2D(UVs=@n38) {parameterName=MaskTexture, texture=/Game/Accessories/FORT_Backpacks/Backpack_M_MED_KeyChain/FX/T_Backpack_KeyChain_Flame}.R) {constB=0.5}
  n40 = Power(Base=Multiply(A=TextureSample(UVs=@n38) {texture=&/T_Pickaxe_SoilBlend_Energy}.R) {constB=3}) {constExponent=5}
  n41 = Multiply(A=Multiply(A=Multiply(A=@n27, B=@n29), B=Saturate(None=@n30)), B=Saturate(None=Add(A=@n39, B=@n40)))
  n42 = ScalarParameter() {defaultValue=0.5, parameterName=OrangeTextureU}
  n43 = ScalarParameter() {defaultValue=0.5, parameterName=OrangeTextureV}
  n44 = Add(A=Multiply(A=TextureCoordinate(), B=AppendVector(A=@n42, B=@n43)), B=Multiply(A=@n34) {constB=0.02})
  n45 = Multiply(A=Constant2Vector() {r=-1}, B=ScalarParameter() {defaultValue=1.5, parameterName=Texture_Speed_Master})
  n46 = Multiply(A=TextureSampleParameter2D(UVs=Panner(Coordinate=@n44, Speed=@n45) {speedX=-1, speedY=-0.05}) {parameterName=Orange_Texture, texture=/Game/Global/Textures/Default/Caustics/T_Caustics_Softer}.R) {constB=1.5}
  n47 = LinearInterpolate(Alpha=TextureSample(UVs=TextureCoordinate() {uTiling=1.2, vTiling=0.3}) {texture=&T_Caustics_Softer}.R) {constB=5}
  n48 = Multiply(
   A=LinearInterpolate(
    A=Multiply(A=@n41, B=@n46)
    B=LinearInterpolate(Alpha=@n41) {constA=-0.2}
    Alpha=ScalarParameter() {defaultValue=1, parameterName=Switch_Orange}
   )
   B=MaterialFunctionCall(Transition Phase=DynamicParameter() {paramNames [Dissolve,U_Scale,U_Offset,Param4]}.Dissolve, Height Texture=Saturate(None=@n47)) {materialFunction=&HeightLerp}.Alpha
  )
  n49 = ScalarParameter() {defaultValue=1, parameterName=Use_SecondTrail}
  n50 = Multiply(
   A=Multiply(A=@n1.A, B=LinearInterpolate(A=@n25, B=@n48, Alpha=@n49))
   B=MaterialFunctionCall("FadeSensitivity (optional)"=ScalarParameter() {defaultValue=0.8, parameterName=MeshFade}) {materialFunction=&FadeBasedOnViewAngle}.Result
  )
  n51 = DepthFade(Opacity=@n50, FadeDistance=ScalarParameter() {defaultValue=5, parameterName=DepthFade})
  ShadingPathSwitch(Default=@n51, Deferred=@n51, Forward=@n50, Mobile=@n50)
 ]
}
