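# Translucent unlit Niagara effect material for the SoilBlend pickaxe, apparently its swing trail, with particle-colored emissive.
# Opacity switches between a noise-distorted, scrolling trail mask and edge/pattern masks, then fades by particle alpha and view angle.
# A panning distortion texture drives local-axis vertex offset, with optional spline thickening.
# Shading-path switches drop depth fade and noise-driven refraction on forward and mobile, while vertex red scales scroll speeds and emissive.
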Material /BRCosmetics/Weapons/FORT_Melee/Pickaxe_SoilBlend/FX/M_Pickaxe_SoilBlend_main
graph {
 properties {
  bAutomaticallySetUsageInEditor=false
  bUsedWithNiagaraMeshParticles=true
  bUsedWithNiagaraRibbons=true
  bUsedWithNiagaraSprites=true
  blendMode=BLEND_Translucent
  refractionMethod=RM_IndexOfRefraction
  shadingModel=MSM_Unlit
  usedShadingModels=Unlit
 }
 nodes [
  Constant()
  n1 = Multiply(
   A=TextureCoordinate()
   B=AppendVector(A=ScalarParameter() {defaultValue=1, parameterName=WPOTexture_U}, B=ScalarParameter() {defaultValue=0.8, parameterName=WPOTexture_V})
  )
  n2 = ScalarParameter() {defaultValue=1, parameterName=Texture_Tile_Master}
  n3 = ScalarParameter() {defaultValue=-1, parameterName=WPOTexture_SpeedX}
  n4 = ScalarParameter() {parameterName=WPOTexture_SpeedY}
  n5 = TextureSampleParameter2D(UVs=Panner(Coordinate=Multiply(A=@n1, B=@n2), Speed=AppendVector(A=@n3, B=@n4)) {speedX=-1}) {parameterName=WPOTex, samplerType=SAMPLERTYPE_Normal, texture=&T_Weap_Distort_N_02}
  n6 = TextureCoordinate()
  n7 = Multiply(A=LinearInterpolate(Alpha=@n5.R) {constB=1.5}, B=LinearInterpolate(Alpha=ComponentMask(None=@n6) {r=true}) {constA=0.05})
  n8 = Multiply(A=Transform(None=Constant3Vector() {constant=<rgba 1 0 0 0>}) {transformSourceType=TRANSFORMSOURCE_Local}, B=@n7)
  n9 = Multiply(A=@n8, B=ScalarParameter() {defaultValue=1, parameterName=X_WPO_Intensity}) {constB=10}
  n10 = Multiply(A=Transform(None=Constant3Vector() {constant=<rgba 0 1 0 0>}) {transformSourceType=TRANSFORMSOURCE_Local}, B=@n7)
  n11 = Multiply(A=@n10, B=ScalarParameter() {defaultValue=1, parameterName=Y_WPO_Intensity}) {constB=10}
  n12 = Multiply(A=Transform(None=Constant3Vector() {constant=<rgba 0 0 1 0>}) {transformSourceType=TRANSFORMSOURCE_Local}, B=@n7)
  n13 = ScalarParameter() {defaultValue=1, parameterName=Z_WPO_Intensity}
  n14 = Multiply(A=ScalarParameter() {defaultValue=5, parameterName=WPO_Intensity}, B=VertexColor().R)
  n15 = Multiply(A=@n8, B=@n14) {constB=10}
  n16 = Multiply(A=@n10, B=@n14) {constB=10}
  n17 = LinearInterpolate(
   A=Add(A=Add(A=@n9, B=@n11), B=Multiply(A=@n12, B=@n13) {constB=10})
   B=Add(A=Add(A=@n15, B=@n16), B=Multiply(A=@n12, B=@n14) {constB=10})
   Alpha=ScalarParameter() {defaultValue=1, parameterName=Use_BreakWPO}
  )
  n18 = ScalarParameter() {defaultValue=5, parameterName=Spline_Base}
  n19 = ScalarParameter() {defaultValue=5, parameterName=Spline_Tip}
  n20 = StaticBool()
  LinearInterpolate(
   A=Add(
    A=@n17
    B=MaterialFunctionCall(WidthBase=@n18, WidthTip=@n19, Expand U or V UV Channel=@n20) {materialFunction=&SplineThicken}.WorldPositionOffset
   )
   B=@n17
   Alpha=ScalarParameter() {defaultValue=1, parameterName=Use_SplineThicken}
  )
  n21 = ParticleColor()
  n22 = VertexColor()
  MaterialFunctionCall(Emissive=Multiply(A=Multiply(A=@n21.RGB, B=@n22.R), B=ScalarParameter() {defaultValue=1, parameterName=All_Emissive})) {materialFunction=&MF_FX_EmissiveAdjustments}
  n23 = DynamicParameter() {defaultValue=<rgba 1 1 1 0.5>, paramNames [PatternNoiseTexU, PatternNoiseTexV, Param3, Refraction]}
  n24 = Multiply(
   A=Add(A=TextureCoordinate(), B=AppendVector(A=@n23.PatternNoiseTexU, B=@n23.PatternNoiseTexU))
   B=ScalarParameter() {defaultValue=0.5, parameterName=PatternNoiseTex_Tile}
  )
  n25 = Multiply(A=Constant2Vector() {r=-0.2}, B=ScalarParameter() {defaultValue=1, parameterName=PatternNoise_speed})
  n26 = ComponentMask(None=TextureSampleParameter2D(UVs=Panner(Coordinate=@n24, Speed=@n25) {speedX=-2}) {parameterName=PatternNoiseTex, texture=/Game/Effects/Fort_Effects/Emotes/UltraLight/Texture/T_SoftNoise_BC7}.RGB) {g=true, r=true}
  n27 = LinearInterpolate(Alpha=ComponentMask(None=TextureCoordinate()) {r=true}) {constA=0.1}
  n28 = Add(
   A=TextureCoordinate()
   B=Multiply(A=Multiply(A=@n26, B=@n27), B=ScalarParameter() {defaultValue=0.01, parameterName=PatternNoise_Intensity}) {constB=0.3}
  )
  n29 = TextureSample(UVs=Panner(Coordinate=TextureCoordinate() {uTiling=0.1, vTiling=0.1}, Time=Constant() {r=0.1}) {speedX=-0.2}) {samplerType=SAMPLERTYPE_LinearColor, texture=/Game/Effects/Fort_Effects/Emotes/UltraLight/Texture/T_Plasma_D_C}
  n30 = Multiply(
   A=Power(
    Base=TextureSampleParameter2D(UVs=@n28) {parameterName=PatternTex, texture=&/T_Pickaxe_SoilBlend_MainMask}.R
    Exp=ScalarParameter() {defaultValue=1, parameterName=Patternpower}
   )
   B=LinearInterpolate(Alpha=AppendVector(A=@n29.R, B=@n29.R)) {constA=-0.2}
  )
  Multiply(A=@n30)
  n31 = Multiply(
   A=TextureCoordinate()
   B=AppendVector(A=ScalarParameter() {defaultValue=1, parameterName=BaseTexture_U}, B=ScalarParameter() {defaultValue=0.8, parameterName=BaseTexture_V})
  )
  n32 = ScalarParameter() {defaultValue=-1.2, parameterName=BaseTextureNoise_SpeedX}
  n33 = ScalarParameter() {parameterName=BaseTextureNoise_SpeedY}
  n34 = TextureSampleParameter2D(UVs=Panner(Coordinate=TextureCoordinate() {uTiling=1.2, vTiling=0.8}, Speed=AppendVector(A=@n32, B=@n33)) {speedX=-1.2}) {parameterName=BaseTextureNoiseTex, texture=/BRCosmetics/Weapons/FORT_Melee/Pickaxe_CoatCheck/FX/T_Pickaxe_CoatCheck_Noise}
  n35 = ComponentMask(None=TextureCoordinate()) {r=true}
  n36 = Multiply(
   A=Multiply(A=ComponentMask(None=@n34.RGB) {g=true, r=true}, B=LinearInterpolate(Alpha=@n35) {constA=0.1})
   B=ScalarParameter() {defaultValue=0.01, parameterName=Noise_Intensity}
  ) {constB=0.3}
  n37 = Multiply(A=ScalarParameter() {defaultValue=-1, parameterName=BaseTexture_SpeedX}, B=VertexColor().R)
  n38 = ScalarParameter() {parameterName=BaseTexture_SpeedY}
  n39 = TextureSampleParameter2D(UVs=Panner(Coordinate=Add(A=@n31, B=@n36), Speed=AppendVector(A=@n37, B=@n38)) {speedX=-1, speedY=-0.1}) {parameterName=Base_Texture, samplerType=SAMPLERTYPE_LinearColor, texture=&T_NoisePack_LinearColor}
  n40 = ScalarParameter() {defaultValue=-2, parameterName=gradient_SpeedX}
  n41 = VertexColor()
  n42 = Panner(
   Coordinate=TextureCoordinate() {uTiling=2, vTiling=0.8}
   Speed=AppendVector(A=Multiply(A=@n40, B=@n41.R), B=ScalarParameter() {parameterName=gradient_SpeedY})
  ) {speedX=-2}
  n43 = Add(
   A=TextureCoordinate() {vTiling=0.45}
   B=Multiply(
    A=TextureSample(UVs=@n42) {samplerType=SAMPLERTYPE_LinearGrayscale, texture=/Game/Environments/Hexmap/HexClouds/T_HexmapFog_01_Height02}.R
    B=Constant() {r=0.03}
   ) {constB=0.3}
  )
  n44 = ComponentMask(None=@n43) {r=true}
  n45 = Saturate(None=Power(Base=Multiply(A=@n44) {constB=20}) {constExponent=1})
  n46 = OneMinus(None=@n44)
  n47 = Saturate(None=Power(Base=Multiply(A=@n46) {constB=5}) {constExponent=1})
  n48 = Power(Base=Multiply(A=@n44) {constB=5}) {constExponent=1.5}
  n49 = ComponentMask(None=@n43) {g=true}
  n50 = Multiply(A=Multiply(A=Multiply(A=@n45, B=@n47), B=Saturate(None=@n48)), B=Power(Base=Sine(None=@n49)) {constExponent=3})
  n51 = Power(
   Base=Multiply(A=ChannelMaskParameter(None=@n39.RGBA) {parameterName=Tex_ChargeRGBA, sortPriority=7}.R, B=@n50)
   Exp=LinearInterpolate() {constA=2}
  )
  n52 = Multiply(
   A=TextureCoordinate()
   B=AppendVector(A=ScalarParameter() {defaultValue=1, parameterName=MaskTextureU}, B=ScalarParameter() {defaultValue=1, parameterName=MaskTextureV})
  )
  n53 = Multiply(A=ScalarParameter() {defaultValue=-1.5, parameterName=MaskTexture_SpeedX}, B=VertexColor().R)
  n54 = ScalarParameter() {parameterName=MaskTexture_SpeedY}
  n55 = TextureSampleParameter2D(UVs=Panner(Coordinate=Add(A=@n52, B=@n36), Speed=AppendVector(A=@n53, B=@n54)) {speedX=-1}) {parameterName=MaskTexture, texture=&/T_Pickaxe_SoilBlend_TrailMask}
  n56 = Power(Base=LinearInterpolate(Alpha=Multiply(A=@n51, B=@n55.R)) {constB=2.5}, Exp=ScalarParameter() {defaultValue=1, parameterName=MaskPower})
  n57 = TextureSample() {texture=&/T_Pickaxe_SoilBlend_EdgeMask}
  n58 = TextureCoordinate() {uTiling=0.5, vTiling=0.5}
  n59 = LinearInterpolate(Alpha=TextureSample(UVs=Panner(Coordinate=@n58) {speedX=-0.2}) {samplerType=SAMPLERTYPE_LinearColor, texture=&T_Plasma_D_C}.R) {constA=1}
  n60 = LinearInterpolate(A=Saturate(None=Multiply(A=@n57.R, B=@n59)), B=@n30, Alpha=ScalarParameter() {defaultValue=1, parameterName=Use_Sharp corner mask})
  n61 = ScalarParameter() {defaultValue=1, parameterName=Switch_MaskGlow}
  n62 = MaterialFunctionCall("FadeSensitivity (optional)"=ScalarParameter() {defaultValue=5, parameterName=MeshFade}) {materialFunction=&FadeBasedOnViewAngle}
  n63 = Multiply(A=Multiply(A=@n21.A, B=LinearInterpolate(A=@n56, B=@n60, Alpha=@n61)), B=Saturate(None=Multiply(A=@n62.Result) {constB=10}))
  n64 = DepthFade(Opacity=@n63, FadeDistance=ScalarParameter() {defaultValue=2, parameterName=DepthFade})
  ShadingPathSwitch(Default=@n64, Deferred=@n64, Forward=@n63, Mobile=@n63)
  n65 = TextureSampleParameter2D(UVs=Panner(Coordinate=TextureCoordinate() {uTiling=0.8, vTiling=0.8}) {speedX=-0.5}) {
   parameterName=RefractionTex
   samplerType=SAMPLERTYPE_LinearGrayscale
   texture=&T_SimplePerlin_Greyscale
  }
  n66 = Constant() {r=1}
  n67 = LinearInterpolate(
   A=LinearInterpolate(
    B=DynamicParameter() {defaultValue=<rgba 1 1 1 0.5>, paramNames [PatternNoiseTexU, PatternNoiseTexV, Param3, Refraction]}.Refraction
    Alpha=Multiply(A=@n65.R, B=@n50)
   ) {constA=1, constB=0.5}
   B=@n66
   Alpha=ScalarParameter() {defaultValue=1, parameterName=Use_Refraction}
  )
  ShadingPathSwitch(Default=@n67, Deferred=@n67, Forward=@n66, Mobile=@n66)
 ]
}
